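# Sound cue played when the Constructor's Bull Rush gadget ability activates.
# A Random node picks between two equally weighted wave variations.
# The cue routes through the Fort_WeaponMelee_PSM sound class at reduced volume, with the shared ability radius attenuation asset controlling spatial falloff.
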SoundCue /Game/Sounds/Fort_Gadgets/ConstructionWorker/Ability_BullsRush_Activate_Cue
root=Random_1
nodes [
 {id=Random_1, children [WavePlayer_1,WavePlayer_2], properties {weights [1,1]}}
 {id=WavePlayer_1, properties {soundWaveAssetPtr=&/Ability_BullsRush_Activate_01}}
 {id=WavePlayer_2, properties {soundWaveAssetPtr=&/Ability_BullsRush_Activate_02}}
]
properties {
 attenuationSettings=/Game/Sounds/Attenuation/Fort_Ability_Radius
 maxDistance=1200
 soundClassObject=&Fort_WeaponMelee_PSM
 volumeMultiplier=0.5
}
